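; Gameplay ability for a glider deployment quest, which reports a custom quest stat event when a gameplay event activates it.
; It calls the parent activation, then gets the quest manager from the avatar actor's Fortnite player controller.
; It sends a complex custom stat event with a count of one, empty source tags and the AdditionalTargetTags variable as target tags, then ends immediately.
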(event Ability|EventActivateAbilityFromEvent
  (params (EventData :type Gameplay%20Event%20Data%20Structure%20%28by%20ref%29))
  (|CallParentFunction EventData)
  (Quest|SendComplexCustomStatEvent (Quests|GetQuestManager (Fortnite|GetFortPlayerControllerfromActor (Ability|GetAvatarActorfromActorInfo)) "Invalid") (Fortnite|GetFortPlayerControllerfromActor (Ability|GetAvatarActorfromActorInfo)) "(GameplayTags=)" (Variables|Default|GetAdditionalTargetTags) 1)
  (Ability|EndAbility))
